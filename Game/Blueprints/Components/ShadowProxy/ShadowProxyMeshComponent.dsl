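; Shadow proxy mesh component Blueprint that exposes a switch for forcing its shadow off.
; Its single custom event, setProxyShadowForceDisabled, passes a Boolean straight through to SetForceShadowDisabled and adds no other logic.
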(event Custom|setProxyShadowForceDisabled
  (params (forceShadowDisabled :type Boolean))
  (CallFunction|SetForceShadowDisabled forceShadowDisabled))
